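# Physics asset for the StrideMiceGiant backpack cosmetic, defining collision and joints for its two lower tubes.
# A single capsule on the pack bone anchors left and right chains of three small capsule bodies each.
# Each chain links segment to segment from the pack outward, with identical joint limits allowing moderate swing and a zero twist limit.
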PhysicsAsset /BRCosmetics/Accessories/FORT_Backpacks/Backpack_M_MED_StrideMiceGiant/Meshes/M_MED_StrideMiceGiant_Pack_Physics
bodies {
 Dyn_LowerTube_L_01 [{shapeType=Capsule, length=1.796, radius=1.2357, shapeName=Dyn_LowerTube_L_01_capsule, center=<2.5034 0 0>, rotation=<rot 90 0 0>}]
 Dyn_LowerTube_L_02 [{shapeType=Capsule, length=1.1571, radius=1.1523, shapeName=Dyn_LowerTube_L_02_capsule, center=<0.6518 0 0>, rotation=<rot 90 0 0>}]
 Dyn_LowerTube_L_03 [{shapeType=Capsule, length=0.5977, radius=1.2498, shapeName=Dyn_LowerTube_L_03_capsule, center=<-0.8002 0 0>, rotation=<rot 90 0 0>}]
 Dyn_LowerTube_R_01 [{shapeType=Capsule, length=1.2079, radius=1.0104, shapeName=Dyn_LowerTube_R_01_capsule, center=<2.54 0 0>, rotation=<rot 90 0 0>}]
 Dyn_LowerTube_R_02 [{shapeType=Capsule, length=2.1698, radius=0.9147, shapeName=Dyn_LowerTube_R_02_capsule, rotation=<rot 90 0 0>}]
 Dyn_LowerTube_R_03 [{shapeType=Capsule, length=1.2345, radius=1.409, shapeName=Dyn_LowerTube_R_03_capsule, center=<-1.7974 0 0>, rotation=<rot 90 0 0>}]
 pack [{shapeType=Capsule, length=5, radius=5, rotation=<rot 90 0 0>}]
}
constraints [
 {bone1Name=Dyn_LowerTube_L_02, bone2Name=Dyn_LowerTube_L_01, swing1LimitDegrees=25, swing2LimitDegrees=45, twistLimitDegrees=0}
 {bone1Name=Dyn_LowerTube_L_03, bone2Name=Dyn_LowerTube_L_02, swing1LimitDegrees=25, swing2LimitDegrees=45, twistLimitDegrees=0}
 {bone1Name=Dyn_LowerTube_R_02, bone2Name=Dyn_LowerTube_R_01, swing1LimitDegrees=25, swing2LimitDegrees=45, twistLimitDegrees=0}
 {bone1Name=Dyn_LowerTube_R_03, bone2Name=Dyn_LowerTube_R_02, swing1LimitDegrees=25, swing2LimitDegrees=45, twistLimitDegrees=0}
 {bone1Name=Dyn_LowerTube_L_01, bone2Name=pack, swing1LimitDegrees=25, swing2LimitDegrees=45, twistLimitDegrees=0}
 {bone1Name=Dyn_LowerTube_R_01, bone2Name=pack, swing1LimitDegrees=25, swing2LimitDegrees=45, twistLimitDegrees=0}
]
